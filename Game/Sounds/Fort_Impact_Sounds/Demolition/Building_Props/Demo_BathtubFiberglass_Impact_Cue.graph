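# Sound cue for impacts on a fiberglass bathtub prop in the demolition impact sound set.
# A Random node picks one of three equally weighted hit samples and feeds a Modulator that slightly varies pitch on each play.
# The cue plays in the Fort_Impacts sound class with small-object demolition attenuation and a reduced volume.
# It joins a concurrency set for object impacts, named for a four-sound cap.
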SoundCue /Game/Sounds/Fort_Impact_Sounds/Demolition/Building_Props/Demo_BathtubFiberglass_Impact_Cue
root=Modulator_2
nodes [
 {id=Modulator_2, children [Random_11], properties {pitchMax=1.07, pitchMin=0.92}}
 {id=Random_11, children [WavePlayer_54, WavePlayer_55, WavePlayer_56], properties {weights [1,1,1]}}
 {id=WavePlayer_54, properties {soundWaveAssetPtr=/Game/Sounds/Fort_Impact_Sounds/Demolition/Bathtub/Bathtub_Fiberglass_Hit04}}
 {id=WavePlayer_55, properties {soundWaveAssetPtr=/Game/Sounds/Fort_Impact_Sounds/Demolition/Bathtub/Bathtub_Fiberglass_Hit01}}
 {id=WavePlayer_56, properties {soundWaveAssetPtr=/Game/Sounds/Fort_Impact_Sounds/Demolition/Bathtub/Bathtub_Fiberglass_Hit02}}
]
properties {
 attenuationSettings=/Game/Sounds/Attenuation/Fort_Demolition_Hit_Small_Sized_Object_Att
 maxDistance=4000
 soundClassObject=&Fort_Impacts
 volumeMultiplier=0.68
 concurrencyOverrides {resolutionRule=PreventNew}
 concurrencySet [&CONC_Object_Impacts_FourMax]
}
