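# Sound cue for bathtub demolition impacts, spatialized with small-object hit attenuation and capped by an object-impact concurrency set that prevents new sounds.
# A quality switch selects between two layered mixes and two branches playing a single pitch-modulated, optimized bathtub impact wave.
# Each mix blends randomized fiberglass shower hits pitched up, toilet hits pitched down, and a ceramic break with a volume fade envelope.
# The second mix uses fewer random variations.
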SoundCue /Game/Sounds/Fort_Impact_Sounds/Demolition/Building_Props/Demo_Bathtub_Cue
root=QualityLevel_0
nodes [
 {
  id=Enveloper_0
  children [WavePlayer_5]
  properties {
   pitchCurve {editorCurveData {keys [{time=0}]}}
   volumeCurve {editorCurveData {keys [
      <at 0 1>
      {arriveTangent=-3.9785, interpMode=RCIM_Cubic, leaveTangent=-3.9785, tangentMode=RCTM_User, time=0.2451, value=0.9137}
      {arriveTangent=-0.6089, leaveTangent=-0.6089, tangentMode=RCTM_User, time=1, value=0}
     ]}}
  }
 }
 {
  id=Enveloper_5
  children [WavePlayer_40]
  properties {
   pitchCurve {editorCurveData {keys [{time=0}]}}
   volumeCurve {editorCurveData {keys [
      <at 0 1>
      {arriveTangent=-3.9785, interpMode=RCIM_Cubic, leaveTangent=-3.9785, tangentMode=RCTM_User, time=0.2451, value=0.9137}
      {arriveTangent=-0.6089, leaveTangent=-0.6089, tangentMode=RCTM_User, time=1, value=0}
     ]}}
  }
 }
 {id=Mixer_0, children [Modulator_10,Modulator_2,Modulator_3], properties {inputVolume [1,0.5,0.6]}}
 {id=Mixer_1, children [Modulator_1,Modulator_4,Modulator_5], properties {inputVolume [1,0.5,0.6]}}
 {id=Modulator_0, children [WavePlayer_1], properties {pitchMax=1.2, pitchMin=0.9}}
 {id=Modulator_1, children [Random_0], properties {pitchMax=1.3, pitchMin=1.1}}
 {id=Modulator_10, children [Random_3], properties {pitchMax=1.3, pitchMin=1.1}}
 {id=Modulator_2, children [Random_12], properties {pitchMax=0.9, pitchMin=0.8}}
 {id=Modulator_3, children [Enveloper_5], properties {pitchMax=1.1, pitchMin=0.5}}
 {id=Modulator_4, children [Random_1], properties {pitchMax=0.9, pitchMin=0.8}}
 {id=Modulator_5, children [Enveloper_0], properties {pitchMax=1.1, pitchMin=0.5}}
 {id=Modulator_6, children [WavePlayer_9], properties {pitchMax=1.2, pitchMin=0.9}}
 {id=QualityLevel_0, children [Mixer_0,Mixer_1,Modulator_0,Modulator_6]}
 {id=Random_0, children [WavePlayer_2,WavePlayer_3], properties {weights [1,1]}}
 {id=Random_1, children [WavePlayer_6,WavePlayer_7], properties {weights [1,1]}}
 {id=Random_12, children [WavePlayer_36, WavePlayer_37, WavePlayer_38], properties {weights [1,1,1]}}
 {id=Random_3, children [WavePlayer_18, WavePlayer_19, WavePlayer_20], properties {weights [1,1,1]}}
 {id=WavePlayer_1, properties {soundWaveAssetPtr=/Game/Sounds/Fort_Impact_Sounds/Demolition/Bathtub/Bathtub_Impact_opt}}
 {id=WavePlayer_18, properties {soundWaveAssetPtr=/Game/Sounds/Fort_Impact_Sounds/Demolition/Shower/Shower_Fiberglass_Hit3}}
 {id=WavePlayer_19, properties {soundWaveAssetPtr=/Game/Sounds/Fort_Impact_Sounds/Demolition/Shower/Shower_Fiberglass_Hit1}}
 {id=WavePlayer_2, properties {soundWaveAssetPtr=&Shower_Fiberglass_Hit3}}
 {id=WavePlayer_20, properties {soundWaveAssetPtr=/Game/Sounds/Fort_Impact_Sounds/Demolition/Shower/Shower_Fiberglass_Hit2}}
 {id=WavePlayer_3, properties {soundWaveAssetPtr=&Shower_Fiberglass_Hit1}}
 {id=WavePlayer_36, properties {soundWaveAssetPtr=/Game/Sounds/Fort_Impact_Sounds/Demolition/Toilet/Toilet_Hit18}}
 {id=WavePlayer_37, properties {soundWaveAssetPtr=/Game/Sounds/Fort_Impact_Sounds/Demolition/Toilet/Toilet_Hit04}}
 {id=WavePlayer_38, properties {soundWaveAssetPtr=/Game/Sounds/Fort_Impact_Sounds/Demolition/Toilet/Toilet_Hit09}}
 {id=WavePlayer_40, properties {soundWaveAssetPtr=&/ceramic_break_new_01}}
 {id=WavePlayer_5, properties {soundWaveAssetPtr=&ceramic_break_new_01}}
 {id=WavePlayer_6, properties {soundWaveAssetPtr=&Toilet_Hit18}}
 {id=WavePlayer_7, properties {soundWaveAssetPtr=&Toilet_Hit04}}
 {id=WavePlayer_9, properties {soundWaveAssetPtr=&Bathtub_Impact_opt}}
]
properties {
 attenuationSettings=/Game/Sounds/Attenuation/Fort_Demolition_Hit_Small_Sized_Object_Att
 maxDistance=4000
 soundClassObject=&Fort_Impacts
 volumeMultiplier=0.68
 concurrencyOverrides {resolutionRule=PreventNew}
 concurrencySet [&CONC_Object_Impacts_FourMax]
}
