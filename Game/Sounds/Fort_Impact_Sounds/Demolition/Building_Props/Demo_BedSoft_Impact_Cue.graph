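# Sound cue for soft bed impacts when demolishing building props.
# A mixer layers a random soft bed hit at full volume with a random general bed hit at half volume, each passed through a modulator for slight pitch variation.
# It plays as an impact-class sound with small-object demolition attenuation, and a concurrency set caps simultaneous object impacts.
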SoundCue /Game/Sounds/Fort_Impact_Sounds/Demolition/Building_Props/Demo_BedSoft_Impact_Cue
root=Mixer_0
nodes [
 {id=Mixer_0, children [Modulator_1,Modulator_2], properties {inputVolume [1,0.5]}}
 {id=Modulator_1, children [Random_9], properties {pitchMax=1.07, pitchMin=0.92}}
 {id=Modulator_2, children [Random_2], properties {pitchMax=1.07, pitchMin=0.92}}
 {id=Random_2, children [WavePlayer_8, WavePlayer_9, WavePlayer_11], properties {weights [1,1,1]}}
 {id=Random_9, children [WavePlayer_48, WavePlayer_49], properties {weights [1,1]}}
 {id=WavePlayer_11, properties {soundWaveAssetPtr=/Game/Sounds/Fort_Impact_Sounds/Demolition/Bed/Bed_Hits9}}
 {id=WavePlayer_48, properties {soundWaveAssetPtr=/Game/Sounds/Fort_Impact_Sounds/Demolition/Bed/Bed_Hit_Soft13}}
 {id=WavePlayer_49, properties {soundWaveAssetPtr=/Game/Sounds/Fort_Impact_Sounds/Demolition/Bed/Bed_Hit_Soft14}}
 {id=WavePlayer_8, properties {soundWaveAssetPtr=/Game/Sounds/Fort_Impact_Sounds/Demolition/Bed/Bed_Hits16}}
 {id=WavePlayer_9, properties {soundWaveAssetPtr=/Game/Sounds/Fort_Impact_Sounds/Demolition/Bed/Bed_Hits17}}
]
properties {
 attenuationSettings=/Game/Sounds/Attenuation/Fort_Demolition_Hit_Small_Sized_Object_Att
 maxDistance=4000
 soundClassObject=&Fort_Impacts
 volumeMultiplier=0.68
 concurrencyOverrides {resolutionRule=PreventNew}
 concurrencySet [&CONC_Object_Impacts_FourMax]
}
