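# Sound cue for the final impact of a couch in the demolition building-prop sounds.
# A random node picks one of three equally weighted Couch_ImpactFinal waves, feeding a modulator that slightly varies pitch.
# It uses medium-sized-object destruction attenuation and the Fort impacts sound class at reduced volume.
# It references a four-max object-impact concurrency set, alongside override values that stop the oldest sound.
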SoundCue /Game/Sounds/Fort_Impact_Sounds/Demolition/Building_Props/Demo_Couch_ImpactFinal_Cue
root=Modulator_2
nodes [
 {id=Modulator_2, children [Random_3], properties {pitchMax=1.07, pitchMin=0.92}}
 {id=Random_3, children [WavePlayer_9, WavePlayer_10, WavePlayer_11], properties {weights [1,1,1]}}
 {id=WavePlayer_10, properties {soundWaveAssetPtr=/Game/Sounds/Fort_Impact_Sounds/Demolition/Couch/Couch_ImpactFinal02}}
 {id=WavePlayer_11, properties {soundWaveAssetPtr=/Game/Sounds/Fort_Impact_Sounds/Demolition/Couch/Couch_ImpactFinal03}}
 {id=WavePlayer_9, properties {soundWaveAssetPtr=/Game/Sounds/Fort_Impact_Sounds/Demolition/Couch/Couch_ImpactFinal01}}
]
properties {
 attenuationSettings=/Game/Sounds/Attenuation/Fort_Demolition_Destroy_Medium_Sized_Object_Att
 maxDistance=4000
 soundClassObject=&Fort_Impacts_PSM
 volumeMultiplier=0.544
 concurrencyOverrides {maxCount=4, resolutionRule=StopOldest}
 concurrencySet [&CONC_Object_Impacts_FourMax]
}
