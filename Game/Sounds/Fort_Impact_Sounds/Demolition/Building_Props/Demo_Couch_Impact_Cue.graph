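# Sound cue for couch impacts during demolition of building props.
# A random node picks one of three equally weighted couch impact waves, then a modulator varies its pitch slightly on each play.
# It routes to the impacts sound class with small-object demolition hit attenuation and a four-voice object impact concurrency set.
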SoundCue /Game/Sounds/Fort_Impact_Sounds/Demolition/Building_Props/Demo_Couch_Impact_Cue
root=Modulator_3
nodes [
 {id=Modulator_3, children [Random_4], properties {pitchMax=1.07, pitchMin=0.92}}
 {id=Random_4, children [WavePlayer_12, WavePlayer_13, WavePlayer_14], properties {weights [1,1,1]}}
 {id=WavePlayer_12, properties {soundWaveAssetPtr=/Game/Sounds/Fort_Impact_Sounds/Demolition/Couch/Couch_Impact03}}
 {id=WavePlayer_13, properties {soundWaveAssetPtr=/Game/Sounds/Fort_Impact_Sounds/Demolition/Couch/Couch_Impact01}}
 {id=WavePlayer_14, properties {soundWaveAssetPtr=/Game/Sounds/Fort_Impact_Sounds/Demolition/Couch/Couch_Impact02}}
]
properties {
 attenuationSettings=/Game/Sounds/Attenuation/Fort_Demolition_Hit_Small_Sized_Object_Att
 maxDistance=4000
 soundClassObject=&Fort_Impacts
 volumeMultiplier=0.68
 concurrencyOverrides {maxCount=4, resolutionRule=StopOldest}
 concurrencySet [&CONC_Object_Impacts_FourMax]
}
