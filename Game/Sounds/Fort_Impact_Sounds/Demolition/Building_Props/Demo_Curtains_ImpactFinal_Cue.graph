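# Sound cue for the final demolition impact of curtain building props.
# A quality-level node picks one of four branches with decreasing variety: a random pick of three curtain impact waves, a pick of two, then two single-wave branches.
# Each branch has its own modulator for slight pitch variation.
# The cue uses medium-object destruction attenuation, the impacts sound class and four-voice impact concurrency.
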SoundCue /Game/Sounds/Fort_Impact_Sounds/Demolition/Building_Props/Demo_Curtains_ImpactFinal_Cue
root=QualityLevel_0
nodes [
 {id=Modulator_0, children [Random_0], properties {pitchMax=1.07, pitchMin=0.92}}
 {id=Modulator_1, children [WavePlayer_3], properties {pitchMax=1.07, pitchMin=0.92}}
 {id=Modulator_2, children [WavePlayer_6], properties {pitchMax=1.07, pitchMin=0.92}}
 {id=Modulator_3, children [Random_5], properties {pitchMax=1.07, pitchMin=0.92}}
 {id=QualityLevel_0, children [Modulator_3,Modulator_0,Modulator_1,Modulator_2]}
 {id=Random_0, children [WavePlayer_0,WavePlayer_1], properties {weights [1,1]}}
 {id=Random_5, children [WavePlayer_15, WavePlayer_16, WavePlayer_17], properties {weights [1,1,1]}}
 {id=WavePlayer_0, properties {soundWaveAssetPtr=/Game/Sounds/Fort_Impact_Sounds/Demolition/Curtains/Curtains_ImpactFinal03}}
 {id=WavePlayer_1, properties {soundWaveAssetPtr=/Game/Sounds/Fort_Impact_Sounds/Demolition/Curtains/Curtains_ImpactFinal01}}
 {id=WavePlayer_15, properties {soundWaveAssetPtr=&Curtains_ImpactFinal03}}
 {id=WavePlayer_16, properties {soundWaveAssetPtr=&Curtains_ImpactFinal01}}
 {id=WavePlayer_17, properties {soundWaveAssetPtr=/Game/Sounds/Fort_Impact_Sounds/Demolition/Curtains/Curtains_ImpactFinal02}}
 {id=WavePlayer_3, properties {soundWaveAssetPtr=&Curtains_ImpactFinal03}}
 {id=WavePlayer_6, properties {soundWaveAssetPtr=&Curtains_ImpactFinal03}}
]
properties {
 attenuationSettings=/Game/Sounds/Attenuation/Fort_Demolition_Destroy_Medium_Sized_Object_Att
 maxDistance=4000
 soundClassObject=&Fort_Impacts_PSM
 volumeMultiplier=0.544
 concurrencyOverrides {maxCount=4, resolutionRule=StopOldest}
 concurrencySet [&CONC_Object_Impacts_FourMax]
}
