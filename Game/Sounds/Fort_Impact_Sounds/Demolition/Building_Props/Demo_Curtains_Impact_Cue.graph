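# Sound cue for curtain impacts in building-prop demolition, classed under Fort_Impacts with small-object hit attenuation.
# A quality-level root selects one of four branches: random among three waves, random between two, or a single wave in the last two slots.
# Every branch adds slight random pitch variation through a modulator.
# A four-max object-impact concurrency set is assigned, alongside override settings that stop the oldest sound.
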SoundCue /Game/Sounds/Fort_Impact_Sounds/Demolition/Building_Props/Demo_Curtains_Impact_Cue
root=QualityLevel_0
nodes [
 {id=Modulator_0, children [Random_0], properties {pitchMax=1.07, pitchMin=0.92}}
 {id=Modulator_1, children [WavePlayer_2], properties {pitchMax=1.1, pitchMin=0.9}}
 {id=Modulator_2, children [WavePlayer_3], properties {pitchMax=1.1, pitchMin=0.9}}
 {id=Modulator_3, children [Random_6], properties {pitchMax=1.07, pitchMin=0.92}}
 {id=QualityLevel_0, children [Modulator_3,Modulator_0,Modulator_1,Modulator_2]}
 {id=Random_0, children [WavePlayer_0,WavePlayer_1], properties {weights [1,1]}}
 {id=Random_6, children [WavePlayer_18, WavePlayer_19, WavePlayer_20], properties {weights [1,1,1]}}
 {id=WavePlayer_0, properties {soundWaveAssetPtr=/Game/Sounds/Fort_Impact_Sounds/Demolition/Curtains/Curtains_Impact04}}
 {id=WavePlayer_1, properties {soundWaveAssetPtr=/Game/Sounds/Fort_Impact_Sounds/Demolition/Curtains/Curtains_Impact02}}
 {id=WavePlayer_18, properties {soundWaveAssetPtr=&Curtains_Impact04}}
 {id=WavePlayer_19, properties {soundWaveAssetPtr=&Curtains_Impact02}}
 {id=WavePlayer_2, properties {soundWaveAssetPtr=&Curtains_Impact04}}
 {id=WavePlayer_20, properties {soundWaveAssetPtr=/Game/Sounds/Fort_Impact_Sounds/Demolition/Curtains/Curtains_Impact01}}
 {id=WavePlayer_3, properties {soundWaveAssetPtr=&Curtains_Impact04}}
]
properties {
 attenuationSettings=/Game/Sounds/Attenuation/Fort_Demolition_Hit_Small_Sized_Object_Att
 maxDistance=4000
 soundClassObject=&Fort_Impacts
 volumeMultiplier=0.68
 concurrencyOverrides {maxCount=4, resolutionRule=StopOldest}
 concurrencySet [&CONC_Object_Impacts_FourMax]
}
