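# Sound cue for impacts on a magazine rack building prop during demolition, built from magazine impact samples.
# A mixer layers a random pick of three samples with a quieter, slightly delayed pick from two others, giving each hit a short follow-up rattle.
# A modulator randomizes the pitch of the combined result.
# It plays in the impacts sound class with small-object demolition attenuation and a concurrency set for object impacts.
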SoundCue /Game/Sounds/Fort_Impact_Sounds/Demolition/Building_Props/Demo_MagazineRack_Impact_Cue
root=Modulator_2
nodes [
 {id=Delay_4, children [Random_22], properties {delayMax=0.15, delayMin=0.08}}
 {id=Mixer_1, children [Random_21,Delay_4], properties {inputVolume [1,0.6]}}
 {id=Modulator_2, children [Mixer_1], properties {pitchMax=1.07, pitchMin=0.9}}
 {id=Random_21, children [WavePlayer_65, WavePlayer_69, WavePlayer_68], properties {weights [1,1,1]}}
 {id=Random_22, children [WavePlayer_66, WavePlayer_70], properties {weights [1,1]}}
 {id=WavePlayer_65, properties {soundWaveAssetPtr=/Game/Sounds/Fort_Impact_Sounds/Demolition/Magazine/Magazine_Impact06}}
 {id=WavePlayer_66, properties {soundWaveAssetPtr=/Game/Sounds/Fort_Impact_Sounds/Demolition/Magazine/Magazine_Impact01}}
 {id=WavePlayer_68, properties {soundWaveAssetPtr=/Game/Sounds/Fort_Impact_Sounds/Demolition/Magazine/Magazine_Impact03}}
 {id=WavePlayer_69, properties {soundWaveAssetPtr=/Game/Sounds/Fort_Impact_Sounds/Demolition/Magazine/Magazine_Impact04}}
 {id=WavePlayer_70, properties {soundWaveAssetPtr=/Game/Sounds/Fort_Impact_Sounds/Demolition/Magazine/Magazine_Impact05}}
]
properties {
 attenuationSettings=/Game/Sounds/Attenuation/Fort_Demolition_Hit_Small_Sized_Object_Att
 maxDistance=4000
 soundClassObject=&Fort_Impacts
 volumeMultiplier=0.68
 concurrencyOverrides {resolutionRule=PreventNew}
 concurrencySet [&CONC_Object_Impacts_FourMax]
}
