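# Sound cue for impacts on a small nightstand prop, part of the demolition building-prop impact sounds.
# A random node picks equally among three nightstand impact waves, then a modulator slightly varies pitch.
# It plays at reduced volume in the impacts sound class with small-object demolition attenuation.
# A shared object-impact concurrency set limits simultaneous instances.
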SoundCue /Game/Sounds/Fort_Impact_Sounds/Demolition/Building_Props/Demo_NightstandSmall_Impact_Cue
root=Modulator_2
nodes [
 {id=Modulator_2, children [Random_9], properties {pitchMax=1.07, pitchMin=0.92}}
 {id=Random_9, children [WavePlayer_61, WavePlayer_63, WavePlayer_64], properties {weights [1,1,1]}}
 {id=WavePlayer_61, properties {soundWaveAssetPtr=/Game/Sounds/Fort_Impact_Sounds/Demolition/Nightstand/Nightstand_Small_Impact01}}
 {id=WavePlayer_63, properties {soundWaveAssetPtr=/Game/Sounds/Fort_Impact_Sounds/Demolition/Nightstand/Nightstand_Small_Impact03}}
 {id=WavePlayer_64, properties {soundWaveAssetPtr=/Game/Sounds/Fort_Impact_Sounds/Demolition/Nightstand/Nightstand_Small_Impact04}}
]
properties {
 attenuationSettings=/Game/Sounds/Attenuation/Fort_Demolition_Hit_Small_Sized_Object_Att
 maxDistance=4000
 soundClassObject=&Fort_Impacts
 volumeMultiplier=0.68
 concurrencyOverrides {resolutionRule=PreventNew}
 concurrencySet [&CONC_Object_Impacts_FourMax]
}
