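# Sound cue for the final demolition impact of a pumpkin building prop.
# It randomly picks one of two equally weighted pumpkin impact waves, then modulates pitch over a wide range and boosts volume with slight variation.
# It plays on the impacts sound class with medium-object destruction attenuation and a four-voice object-impact concurrency limit.
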SoundCue /Game/Sounds/Fort_Impact_Sounds/Demolition/Building_Props/Demo_Pumpkin_ImpactFinal_Cue
root=Modulator_4
nodes [
 {id=Modulator_4, children [Random_1], properties {pitchMax=1.1, pitchMin=0.7, volumeMax=1.6, volumeMin=1.4}}
 {id=Random_1, children [WavePlayer_4,WavePlayer_5], properties {weights [1,1]}}
 {id=WavePlayer_4, properties {soundWaveAssetPtr=/Game/Sounds/Fort_Impact_Sounds/Demolition/Pumpkin/Pumpkin_Impact_01}}
 {id=WavePlayer_5, properties {soundWaveAssetPtr=/Game/Sounds/Fort_Impact_Sounds/Demolition/Pumpkin/Pumpkin_Impact_02}}
]
properties {
 attenuationSettings=/Game/Sounds/Attenuation/Fort_Demolition_Destroy_Medium_Sized_Object_Att
 maxDistance=4000
 soundClassObject=&Fort_Impacts_PSM
 volumeMultiplier=0.8
 concurrencyOverrides {maxCount=4, resolutionRule=StopOldest}
 concurrencySet [&CONC_Object_Impacts_FourMax]
}
