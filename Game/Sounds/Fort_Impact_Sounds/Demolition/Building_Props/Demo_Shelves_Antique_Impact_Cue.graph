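# Impact sound cue for demolishing antique shelves, routed to the impacts sound class with small-object hit attenuation.
# A Quality Level node selects one of four branches, which step down from four random hit waves to three, two and a single wave.
# Each branch runs through a Modulator for slight pitch variation.
# A four-max object-impact concurrency group is assigned to limit simultaneous hits.
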SoundCue /Game/Sounds/Fort_Impact_Sounds/Demolition/Building_Props/Demo_Shelves_Antique_Impact_Cue
root=QualityLevel_0
nodes [
 {id=Modulator_0, children [Random_0], properties {pitchMax=1.07, pitchMin=0.92}}
 {id=Modulator_1, children [Random_1], properties {pitchMax=1.1, pitchMin=0.9}}
 {id=Modulator_2, children [WavePlayer_5], properties {pitchMax=1.1, pitchMin=0.9}}
 {id=Modulator_6, children [Random_19], properties {pitchMax=1.07, pitchMin=0.92}}
 {id=QualityLevel_0, children [Modulator_6,Modulator_0,Modulator_1,Modulator_2]}
 {id=Random_0, children [WavePlayer_0,WavePlayer_1,WavePlayer_2], properties {weights [1,1,1]}}
 {id=Random_1, children [WavePlayer_3,WavePlayer_4], properties {bRandomizeWithoutReplacement=false, weights [1,1]}}
 {id=Random_19, children [WavePlayer_79, WavePlayer_80, WavePlayer_82, WavePlayer_83], properties {weights [1,1,1,1]}}
 {id=WavePlayer_0, properties {soundWaveAssetPtr=/Game/Sounds/Fort_Impact_Sounds/Demolition/Shelves/Shelves_Antique_Hit6}}
 {id=WavePlayer_1, properties {soundWaveAssetPtr=/Game/Sounds/Fort_Impact_Sounds/Demolition/Shelves/Shelves_Antique_Hit1}}
 {id=WavePlayer_2, properties {soundWaveAssetPtr=/Game/Sounds/Fort_Impact_Sounds/Demolition/Shelves/Shelves_Antique_Hit3}}
 {id=WavePlayer_3, properties {soundWaveAssetPtr=&Shelves_Antique_Hit6}}
 {id=WavePlayer_4, properties {soundWaveAssetPtr=&Shelves_Antique_Hit1}}
 {id=WavePlayer_5, properties {soundWaveAssetPtr=&Shelves_Antique_Hit6}}
 {id=WavePlayer_79, properties {soundWaveAssetPtr=&Shelves_Antique_Hit6}}
 {id=WavePlayer_80, properties {soundWaveAssetPtr=&Shelves_Antique_Hit1}}
 {id=WavePlayer_82, properties {soundWaveAssetPtr=&Shelves_Antique_Hit3}}
 {id=WavePlayer_83, properties {soundWaveAssetPtr=/Game/Sounds/Fort_Impact_Sounds/Demolition/Shelves/Shelves_Antique_Hit4}}
]
properties {
 attenuationSettings=/Game/Sounds/Attenuation/Fort_Demolition_Hit_Small_Sized_Object_Att
 maxDistance=4000
 soundClassObject=&Fort_Impacts
 volumeMultiplier=0.68
 concurrencyOverrides {resolutionRule=PreventNew}
 concurrencySet [&CONC_Object_Impacts_FourMax]
}
