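# Sound cue for impacts on bottle-laden shelf props during demolition.
# A mixer layers three pitched-up branches: a single quiet canned-food hit, a random pick of three canned-food hits, and a random pick of four bottle impacts.
# Per-branch pitch modulation adds variation on each play.
# It uses small-object demolition attenuation, the impacts sound class and an object-impact concurrency set that limits simultaneous instances.
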SoundCue /Game/Sounds/Fort_Impact_Sounds/Demolition/Building_Props/Demo_Shelves_Bottles_Impact_Cue
root=Mixer_1
nodes [
 {id=Mixer_1, children [Modulator_2,Modulator_1,Modulator_7], properties {inputVolume [0.25,1,0.8]}}
 {id=Modulator_1, children [Random_30], properties {pitchMax=1.4, pitchMin=1.2}}
 {id=Modulator_2, children [WavePlayer_198], properties {pitchMax=1.4, pitchMin=1.2}}
 {id=Modulator_7, children [Random_17], properties {pitchMax=1.3, pitchMin=1.1}}
 {id=Random_17, children [WavePlayer_190, WavePlayer_191, WavePlayer_195, WavePlayer_196], properties {weights [1,1,1,1]}}
 {id=Random_30, children [WavePlayer_199, WavePlayer_200, WavePlayer_201], properties {weights [1,1,1]}}
 {id=WavePlayer_190, properties {soundWaveAssetPtr=/Game/Sounds/Fort_Impact_Sounds/Demolition/Shelves/Shelf_Bottle_Impact02}}
 {id=WavePlayer_191, properties {soundWaveAssetPtr=/Game/Sounds/Fort_Impact_Sounds/Demolition/Shelves/Shelf_Bottle_Impact03}}
 {id=WavePlayer_195, properties {soundWaveAssetPtr=/Game/Sounds/Fort_Impact_Sounds/Demolition/Shelves/Shelf_Bottle_Impact07}}
 {id=WavePlayer_196, properties {soundWaveAssetPtr=/Game/Sounds/Fort_Impact_Sounds/Demolition/Shelves/Shelf_Bottle_Impact08}}
 {id=WavePlayer_198, properties {soundWaveAssetPtr=/Game/Sounds/Fort_Impact_Sounds/Demolition/Shelves/Shelves_CanFood_Hit12}}
 {id=WavePlayer_199, properties {soundWaveAssetPtr=/Game/Sounds/Fort_Impact_Sounds/Demolition/Shelves/Shelves_CanFood_Hit17}}
 {id=WavePlayer_200, properties {soundWaveAssetPtr=/Game/Sounds/Fort_Impact_Sounds/Demolition/Shelves/Shelves_CanFood_Hit18}}
 {id=WavePlayer_201, properties {soundWaveAssetPtr=/Game/Sounds/Fort_Impact_Sounds/Demolition/Shelves/Shelves_CanFood_Hit4}}
]
properties {
 attenuationSettings=/Game/Sounds/Attenuation/Fort_Demolition_Hit_Small_Sized_Object_Att
 maxDistance=4000
 soundClassObject=&Fort_Impacts
 volumeMultiplier=0.68
 concurrencyOverrides {resolutionRule=PreventNew}
 concurrencySet [&CONC_Object_Impacts_FourMax]
}
