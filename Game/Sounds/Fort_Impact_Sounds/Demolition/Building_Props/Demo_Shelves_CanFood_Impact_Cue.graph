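# Sound cue for impacts on shelves of canned food during demolition.
# A mixer layers two branches, each randomly picking one of two hit waves and varying pitch and volume through a modulator.
# It plays in the impacts sound class with small-object demolition attenuation and a four-max object impact concurrency set.
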SoundCue /Game/Sounds/Fort_Impact_Sounds/Demolition/Building_Props/Demo_Shelves_CanFood_Impact_Cue
root=Mixer_0
nodes [
 {id=Mixer_0, children [Modulator_4,Modulator_3], properties {inputVolume [1,1]}}
 {id=Modulator_3, children [Random_17], properties {pitchMax=1.1, pitchMin=0.9, volumeMin=0.8}}
 {id=Modulator_4, children [Random_16], properties {pitchMax=1.1, pitchMin=0.8, volumeMin=0.7}}
 {id=Random_16, children [WavePlayer_62, WavePlayer_63], properties {bRandomizeWithoutReplacement=false, weights [1,1]}}
 {id=Random_17, children [WavePlayer_72, WavePlayer_65], properties {weights [1,1]}}
 {id=WavePlayer_62, properties {soundWaveAssetPtr=/Game/Sounds/Fort_Impact_Sounds/Demolition/Shelves/Shelves_CanFood_Hit4}}
 {id=WavePlayer_63, properties {soundWaveAssetPtr=/Game/Sounds/Fort_Impact_Sounds/Demolition/Shelves/Shelves_CanFood_Hit12}}
 {id=WavePlayer_65, properties {soundWaveAssetPtr=/Game/Sounds/Fort_Impact_Sounds/Demolition/Shelves/Shelves_CanFood_Hit18}}
 {id=WavePlayer_72, properties {soundWaveAssetPtr=/Game/Sounds/Fort_Impact_Sounds/Demolition/Shelves/Shelves_CanFood_Hit17}}
]
properties {
 attenuationSettings=/Game/Sounds/Attenuation/Fort_Demolition_Hit_Small_Sized_Object_Att
 maxDistance=4000
 soundClassObject=&Fort_Impacts
 volumeMultiplier=0.68
 concurrencyOverrides {resolutionRule=PreventNew}
 concurrencySet [&CONC_Object_Impacts_FourMax]
}
