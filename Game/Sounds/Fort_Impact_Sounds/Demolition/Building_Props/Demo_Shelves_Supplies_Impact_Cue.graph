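# Sound cue for impacts on supply shelves, part of the demolition building-prop impact sounds.
# It picks one of four shelf-hit waves with equal weight, then a modulator slightly randomizes pitch.
# Playback is scaled down in volume, uses small-object demolition attenuation and the impacts sound class, and joins a four-max object-impact concurrency set.
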SoundCue /Game/Sounds/Fort_Impact_Sounds/Demolition/Building_Props/Demo_Shelves_Supplies_Impact_Cue
root=Modulator_10
nodes [
 {id=Modulator_10, children [Random_10], properties {pitchMax=1.07, pitchMin=0.92}}
 {id=Random_10, children [WavePlayer_43, WavePlayer_44, WavePlayer_45, WavePlayer_46], properties {weights [1,1,1,1]}}
 {id=WavePlayer_43, properties {soundWaveAssetPtr=/Game/Sounds/Fort_Impact_Sounds/Demolition/Shelves/Shelves_Supplies_Hit32}}
 {id=WavePlayer_44, properties {soundWaveAssetPtr=/Game/Sounds/Fort_Impact_Sounds/Demolition/Shelves/Shelves_Supplies_Hit1}}
 {id=WavePlayer_45, properties {soundWaveAssetPtr=/Game/Sounds/Fort_Impact_Sounds/Demolition/Shelves/Shelves_Supplies_Hit29}}
 {id=WavePlayer_46, properties {soundWaveAssetPtr=/Game/Sounds/Fort_Impact_Sounds/Demolition/Shelves/Shelves_Supplies_Hit31}}
]
properties {
 attenuationSettings=/Game/Sounds/Attenuation/Fort_Demolition_Hit_Small_Sized_Object_Att
 maxDistance=4000
 soundClassObject=&Fort_Impacts
 volumeMultiplier=0.68
 concurrencyOverrides {resolutionRule=PreventNew}
 concurrencySet [&CONC_Object_Impacts_FourMax]
}
