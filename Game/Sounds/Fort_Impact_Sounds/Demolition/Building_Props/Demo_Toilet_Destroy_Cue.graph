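# Sound cue for destroying a toilet prop, using medium-object demolition attenuation and impact concurrency that prevents new sounds at the limit.
# A quality switch picks the richest branch first: a pitch-varied ceramic break mixed with one of three random toilet hits.
# The second level drops to two hit variants, and the lowest two share a single optimized impact wave.
# Each branch adds slight overall pitch randomization.
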SoundCue /Game/Sounds/Fort_Impact_Sounds/Demolition/Building_Props/Demo_Toilet_Destroy_Cue
root=QualityLevel_0
nodes [
 {id=Mixer_0, children [Modulator_3,Random_0], properties {inputVolume [1,1]}}
 {id=Mixer_3, children [Modulator_4,Random_11], properties {inputVolume [1,1]}}
 {id=Modulator_0, children [WavePlayer_0], properties {pitchMax=1.07, pitchMin=0.92}}
 {id=Modulator_1, children [Mixer_3], properties {pitchMax=1.07, pitchMin=0.92}}
 {id=Modulator_2, children [Mixer_0], properties {pitchMax=1.07, pitchMin=0.92}}
 {id=Modulator_3, children [WavePlayer_3], properties {pitchMax=1.1, pitchMin=0.8}}
 {id=Modulator_4, children [WavePlayer_41], properties {pitchMax=1.1, pitchMin=0.8}}
 {id=QualityLevel_0, children [Modulator_1,Modulator_2,Modulator_0,Modulator_0]}
 {id=Random_0, children [WavePlayer_1,WavePlayer_2], properties {weights [1,1]}}
 {id=Random_11, children [WavePlayer_33, WavePlayer_34, WavePlayer_35], properties {weights [1,1,1]}}
 {id=WavePlayer_0, properties {soundWaveAssetPtr=/Game/Sounds/Fort_Impact_Sounds/Demolition/Toilet/Toilet_Impact_opt}}
 {id=WavePlayer_1, properties {soundWaveAssetPtr=/Game/Sounds/Fort_Impact_Sounds/Demolition/Toilet/Toilet_Hit18}}
 {id=WavePlayer_2, properties {soundWaveAssetPtr=/Game/Sounds/Fort_Impact_Sounds/Demolition/Toilet/Toilet_Hit04}}
 {id=WavePlayer_3, properties {soundWaveAssetPtr=&/ceramic_break_new_01}}
 {id=WavePlayer_33, properties {soundWaveAssetPtr=&Toilet_Hit18}}
 {id=WavePlayer_34, properties {soundWaveAssetPtr=&Toilet_Hit04}}
 {id=WavePlayer_35, properties {soundWaveAssetPtr=/Game/Sounds/Fort_Impact_Sounds/Demolition/Toilet/Toilet_Hit09}}
 {id=WavePlayer_41, properties {soundWaveAssetPtr=&ceramic_break_new_01}}
]
properties {
 attenuationSettings=/Game/Sounds/Attenuation/Fort_Demolition_Destroy_Medium_Sized_Object_Att
 maxDistance=4000
 soundClassObject=&Fort_Impacts
 volumeMultiplier=0.476
 concurrencyOverrides {resolutionRule=PreventNew}
 concurrencySet [&CONC_Object_Impacts_FourMax]
}
