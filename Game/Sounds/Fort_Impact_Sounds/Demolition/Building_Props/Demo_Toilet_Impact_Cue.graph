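# Sound cue for toilet impacts during demolition, played through small-object attenuation with concurrency that blocks new instances at the limit.
# A quality switch picks the variant: the first two levels mix a pitch-varied, volume-enveloped ceramic break with a random toilet hit, drawing on three and two waves respectively.
# The last two levels share one cheaper branch, a single optimized impact wave with a volume fade and slight upward pitch variation.
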SoundCue /Game/Sounds/Fort_Impact_Sounds/Demolition/Building_Props/Demo_Toilet_Impact_Cue
root=QualityLevel_0
nodes [
 {
  id=Enveloper_0
  children [WavePlayer_0]
  properties {
   pitchCurve {editorCurveData {keys [{time=0}]}}
   volumeCurve {editorCurveData {keys [
      {arriveTangent=-3.4098, interpMode=RCIM_Cubic, leaveTangent=-3.4098, tangentMode=RCTM_User, time=0, value=1}
      {interpMode=RCIM_Cubic, leaveTangent=-0.6089, time=0.6, value=0}
     ]}}
  }
 }
 {
  id=Enveloper_1
  children [WavePlayer_3]
  properties {
   pitchCurve {editorCurveData {keys [{time=0}]}}
   volumeCurve {editorCurveData {keys [
      <at 0 1>
      {arriveTangent=-3.9785, interpMode=RCIM_Cubic, leaveTangent=-3.9785, tangentMode=RCTM_User, time=0.2451, value=0.9137}
      {arriveTangent=-0.6089, leaveTangent=-0.6089, tangentMode=RCTM_User, time=0.7, value=0}
     ]}}
  }
 }
 {
  id=Enveloper_6
  children [WavePlayer_41]
  properties {
   pitchCurve {editorCurveData {keys [{time=0}]}}
   volumeCurve {editorCurveData {keys [
      <at 0 1>
      {arriveTangent=-3.9785, interpMode=RCIM_Cubic, leaveTangent=-3.9785, tangentMode=RCTM_User, time=0.2451, value=0.9137}
      {arriveTangent=-0.6089, leaveTangent=-0.6089, tangentMode=RCTM_User, time=0.7, value=0}
     ]}}
  }
 }
 {id=Mixer_0, children [Modulator_3,Random_0], properties {inputVolume [0.5,1.3]}}
 {id=Mixer_3, children [Modulator_4,Random_11], properties {inputVolume [0.5,1.3]}}
 {id=Modulator_0, children [Enveloper_0], properties {pitchMax=1.15, pitchMin=1.05}}
 {id=Modulator_1, children [Mixer_3], properties {pitchMax=1.07, pitchMin=0.92}}
 {id=Modulator_2, children [Mixer_0], properties {pitchMax=1.07, pitchMin=0.92}}
 {id=Modulator_3, children [Enveloper_1], properties {pitchMax=1.1, pitchMin=0.5}}
 {id=Modulator_4, children [Enveloper_6], properties {pitchMax=1.1, pitchMin=0.5}}
 {id=QualityLevel_0, children [Modulator_1,Modulator_2,Modulator_0,Modulator_0]}
 {id=Random_0, children [WavePlayer_1,WavePlayer_2], properties {weights [1,1]}}
 {id=Random_11, children [WavePlayer_33, WavePlayer_34, WavePlayer_35], properties {weights [1,1,1]}}
 {id=WavePlayer_0, properties {soundWaveAssetPtr=/Game/Sounds/Fort_Impact_Sounds/Demolition/Toilet/Toilet_Impact_opt}}
 {id=WavePlayer_1, properties {soundWaveAssetPtr=/Game/Sounds/Fort_Impact_Sounds/Demolition/Toilet/Toilet_Hit18}}
 {id=WavePlayer_2, properties {soundWaveAssetPtr=/Game/Sounds/Fort_Impact_Sounds/Demolition/Toilet/Toilet_Hit04}}
 {id=WavePlayer_3, properties {soundWaveAssetPtr=&/ceramic_break_new_01}}
 {id=WavePlayer_33, properties {soundWaveAssetPtr=&Toilet_Hit18}}
 {id=WavePlayer_34, properties {soundWaveAssetPtr=&Toilet_Hit04}}
 {id=WavePlayer_35, properties {soundWaveAssetPtr=/Game/Sounds/Fort_Impact_Sounds/Demolition/Toilet/Toilet_Hit09}}
 {id=WavePlayer_41, properties {soundWaveAssetPtr=&ceramic_break_new_01}}
]
properties {
 attenuationSettings=/Game/Sounds/Attenuation/Fort_Demolition_Hit_Small_Sized_Object_Att
 maxDistance=4000
 soundClassObject=&Fort_Impacts
 volumeMultiplier=0.68
 concurrencyOverrides {resolutionRule=PreventNew}
 concurrencySet [&CONC_Object_Impacts_FourMax]
}
